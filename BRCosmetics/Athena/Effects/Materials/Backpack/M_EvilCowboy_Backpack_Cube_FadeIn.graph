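# Unlit translucent mesh-particle material, named as the fade-in for the Evil Cowboy backpack cube effect.
# Particle color RGB feeds an emissive adjustment function.
# Particle alpha multiplies a vertex-normal Fresnel term to form an opacity-style mask.
# The first dynamic parameter drives the Fresnel base reflect fraction, letting the particle system move the mask from rim-only toward full coverage.
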Material /BRCosmetics/Athena/Effects/Materials/Backpack/M_EvilCowboy_Backpack_Cube_FadeIn
graph {
 properties {
  bUsedWithMeshParticles=true
  bUsedWithNiagaraMeshParticles=true
  blendMode=BLEND_Translucent
  pixelDepthOffsetMode=PDOM_Legacy
  previewMesh=/BRCosmetics/Athena/Effects/Meshes/Backpack/SM_EvilCowboy_BackPack_Cube
  shadingModel=MSM_Unlit
  usedShadingModels=Unlit
 }
 nodes [
  n1 = ParticleColor()
  MaterialFunctionCall(Emissive=@n1.RGB) {materialFunction=&MF_FX_EmissiveAdjustments}
  Multiply(A=@n1.A, B=Fresnel(BaseReflectFractionIn=DynamicParameter() {defaultValue=<rgba 0 1 1 1>}.Param1, Normal=VertexNormalWS()))
 ]
}
