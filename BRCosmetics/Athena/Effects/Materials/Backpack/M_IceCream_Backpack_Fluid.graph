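# Masked subsurface material for mesh particles, evidently the fluid effect of the ice cream backpack cosmetic.
# Vertex motion adds a normal push, scaled by the Thickness dynamic parameter, to MF_MeltingVertexAnimation's noise-driven offset.
# Particle color feeds an emissive adjustment function.
# The graph also holds a local-space MaskPosition sphere mask and an ice shading chain with normal-perturbed Fresnel, gradient lookup and time-of-day tints.
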Material /BRCosmetics/Athena/Effects/Materials/Backpack/M_IceCream_Backpack_Fluid
graph {
 properties {
  bUsedWithMeshParticles=true
  blendMode=BLEND_Masked
  pixelDepthOffsetMode=PDOM_Legacy
  shadingModel=MSM_Subsurface
  usedShadingModels=Subsurface
 }
 nodes [
  ActorPositionWS()
  n1 = TextureObjectParameter() {
   group=VertexAnim
   parameterName=VertexAnimNormalMain
   samplerType=SAMPLERTYPE_Normal
   texture=/Game/Effects/Fort_Effects/Textures/T_DustDistortion
  }
  Add(
   A=Multiply(A=VertexNormalWS(), B=DynamicParameter() {paramNames [Thickness,Param2,Param3,Param4]}.Thickness)
   B=MaterialFunctionCall!(
    VertexAnimNormalMain=@n1
    VertexAnimTexSize=ScalarParameter() {defaultValue=35, group=VertexAnim, parameterName=VertexAnimTexSize}
    VertexAnimNormalSpeed=ScalarParameter() {defaultValue=35, group=VertexAnim, parameterName=VertexAnimNormalSpeed}
    WPO_PanSpeedWhileMoving=ScalarParameter() {defaultValue=1, group=VertexAnim, parameterName=WPO_PanSpeedWhileMoving}
    UseParticlePositionOrObjectPosition=StaticBoolParameter() {defaultValue=true, parameterName=UseParticlePosition}
    VertexAnimIntesityXY=ScalarParameter() {defaultValue=5, group=VertexAnim, parameterName=VertexAnimIntesityXY}
    WPO_Intensity=ScalarParameter() {defaultValue=0.1, group=VertexAnim, parameterName=WPO_Intensity}
    VertexAnimVerticalMaskBias=ScalarParameter() {defaultValue=0.25, group=VertexAnim, parameterName=VertexAnimVerticalMaskBias}
   ) {materialFunction=/Game/Effects/Fort_Effects/Materials/_Functions/MF_MeltingVertexAnimation}."VertexAnimation(WPO)"
  )
  n2 = ComponentMask(None=Transform(None=WorldPosition().XYZ) {transformSourceType=TRANSFORMSOURCE_World, transformType=TRANSFORM_Local}) {g=true}
  n3 = Normalize(VectorInput=Transform(None=VectorParameter() {parameterName=MaskPosition, defaultValue=<rgba 0 0 0 1>}.RGB) {transformSourceType=TRANSFORMSOURCE_World, transformType=TRANSFORM_Local})
  n4 = Subtract(A=@n2, B=@n3)
  Clamp(None=@n4)
  Floor(None=@n4)
  n5 = TextureSampleParameter2D(UVs=Multiply(A=TextureCoordinate() {uTiling=0.25}, B=ScalarParameter() {defaultValue=1, group=Ice, parameterName=Ice Normal Tiling})) {group=Ice, parameterName=Ice Normal Tiling Tex, samplerType=SAMPLERTYPE_Normal, texture=/Game/Weapons/WeaponSkins/Textures/StoneNormal}
  n6 = OneMinus(None=ScalarParameter() {defaultValue=0.5, group=Ice, parameterName=IceNormalIntensity})
  n7 = ScalarParameter() {
   defaultValue=0.8
   desc=Lerp that controls the blend between the tiling ice normal and the mesh normal texture. Alters normal influence on Ice base/emissive color Fresnel.
   group=Ice
   parameterName=Tiling Normal Influence on Ice Color
  }
  n8 = LinearInterpolate(
   A=VertexNormalWS()
   B=MaterialFunctionCall(Normal=@n5.RGB, Flatness=Saturate(None=@n6)) {materialFunction=&FlattenNormal}.Result
   Alpha=@n7
  )
  n9 = Saturate(None=Fresnel(ExponentIn=ScalarParameter() {defaultValue=25, group=Ice, parameterName=Crystal_FresEX}, Normal=Transform(None=@n8)) {baseReflectFraction=0})
  LinearInterpolate(
   A=Multiply(
    A=TextureSampleParameter2D(UVs=@n9) {group=Ice, parameterName=IceGradient, texture=/Game/Global/Textures/Default/Masks_and_Gradients/T_IceGradient}.RGB
    B=ScalarParameter() {defaultValue=0.8, group=Ice, parameterName=ice Brightness}
   )
   B=MaterialFunctionCall(
    Color_Morning=VectorParameter() {group=Ice, parameterName=Ice Edge Tint Morning, defaultValue=<rgba 0.2545 0.8087 1 1>}.RGB
    Color_Day=VectorParameter() {group=Ice, parameterName=Ice Edge Tint Day, defaultValue=<rgba 0.4136 1 0.8797 1>}.RGB
    Color_Evening=VectorParameter() {group=Ice, parameterName=Ice Edge Tint Evening, defaultValue=<rgba 0 1 0.85 1>}.RGB
    Color_Night=VectorParameter() {group=Ice, parameterName=Ice Edge Tint Night, defaultValue=<rgba 0 0.92 1.5 1>}.RGB
   ) {materialFunction=&Color_Time_of_Day}.Result
  )
  MaterialFunctionCall(Emissive=ParticleColor().RGB) {materialFunction=&MF_FX_EmissiveAdjustments}
  ScalarParameter() {defaultValue=0.4, parameterName=Metallic}
  ScalarParameter() {defaultValue=1, parameterName=Specular}
  ScalarParameter() {defaultValue=1, parameterName=Thickness}
  ScalarParameter() {defaultValue=0.04, parameterName=Roughness}
  SphereMask(A=@n2, B=@n3, Radius=ScalarParameter() {parameterName=Mask_Radius}, Hardness=ScalarParameter() {defaultValue=1, parameterName=Hardness})
  TextureCoordinate()
  VectorParameter() {parameterName=Subsurface Color, defaultValue=<rgba 0 0.01 0.01 0>}
 ]
}
